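# Post-process AnimBlueprint for the male medium PigeonChart cosmetic head skeleton, applying fixed corrections to eye, teeth and tongue bones.
# In component space, each pupilpush bone partially copies its eye bone's translation and rotation, then takes a small additive bone-space offset.
# Additive Transform Bone nodes then shift and scale upper and lower teeth and scale the tongue.
# All values are constants; the EventGraph has no connected logic.
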AnimBlueprint /BRCosmetics/Characters/Player/Male/Medium/Heads/M_MED_PigeonChart_Head/Meshes/M_MED_PigeonChart_Head_Skeleton_PostProcessBlueprint
target_skeleton=&/M_MED_PigeonChart_Head_Skeleton
graphs [
 AnimGraph {nodes [
   n1 = Local To Component(LocalPose=Input Pose())
   n2 = "Transform (Modify) Bone - Bone: L_eye_pupilpush"(
    ComponentPose=Copy Bone - Source Bone: R_eye - Target Bone: R_eye_pupilpush!(
     ComponentPose:Component Space Pose Link Structure=Copy Bone - Source Bone: L_eye - Target Bone: L_eye_pupilpush!(
      ComponentPose:Component Space Pose Link Structure=@n1
      Alpha:float=0.7
      bCopyTranslation:bool=true
      bCopyRotation:bool=true
      bCopyScale:bool=False
     ) -> (Pose:Component Space Pose Link Structure) {is=title AnimGraphNode_CopyBone, alphaBoolBlend {blendOption=Linear}, sourceBone {boneName=L_eye}, targetBone {boneName=L_eye_pupilpush}}
     Alpha:float=0.7
     bCopyTranslation:bool=true
     bCopyRotation:bool=true
     bCopyScale:bool=False
    ) -> (Pose:Component Space Pose Link Structure) {is=title AnimGraphNode_CopyBone, alphaBoolBlend {blendOption=Linear}, sourceBone {boneName=R_eye}, targetBone {boneName=R_eye_pupilpush}}
    Alpha=1
    Translation="0,-0.2,0"
    Rotation="0,0,0"
    Scale="1,1,1"
   ) {
    translationMode=BMM_Additive
    translationSpace=BCS_BoneSpace
    alphaBoolBlend {blendOption=Linear}
    boneToModify {boneName=L_eye_pupilpush}
    translation {y=-0.2}
   }
   n3 = "Transform (Modify) Bone - Bone: R_eye_pupilpush"(ComponentPose=@n2, Alpha=1, Translation="0,-0.2,0", Rotation="0,0,0", Scale="1,1,1") {
    translationMode=BMM_Additive
    translationSpace=BCS_BoneSpace
    alphaBoolBlend {blendOption=Linear}
    boneToModify {boneName=R_eye_pupilpush}
    translation {y=-0.2}
   }
   n4 = "Transform (Modify) Bone - Bone: teeth_upper"(ComponentPose=@n3, Alpha=1, Translation="0,0,0.2", Rotation="0,0,0", Scale="1.3,1,1") {
    scaleMode=BMM_Additive
    scaleSpace=BCS_BoneSpace
    translationMode=BMM_Additive
    alphaBoolBlend {blendOption=Linear}
    boneToModify {boneName=teeth_upper}
    scale {x=1.3}
    translation {z=0.2}
   }
   n5 = "Transform (Modify) Bone - Bone: teeth_lower"(ComponentPose=@n4, Alpha=1, Translation="0,0,0.2", Rotation="0,0,0", Scale="1.3,1,1") {
    scaleMode=BMM_Additive
    scaleSpace=BCS_BoneSpace
    translationMode=BMM_Additive
    alphaBoolBlend {blendOption=Linear}
    boneToModify {boneName=teeth_lower}
    scale {x=1.3}
    translation {z=0.2}
   }
   n6 = "Transform (Modify) Bone - Bone: tongue"(ComponentPose=@n5, Alpha=1, Translation="0,0,0", Rotation="0,0,0", Scale="1.1,1,1") {
    scaleMode=BMM_Additive
    scaleSpace=BCS_BoneSpace
    translationMode=BMM_Additive
    alphaBoolBlend {blendOption=Linear}
    boneToModify {boneName=tongue}
    scale {x=1.1}
   }
   Output Pose(Result=Component To Local(ComponentPose=@n6))
  ]}
 EventGraph {nodes [
   TryGetPawnOwner()
   Event BlueprintUpdateAnimation()
  ]}
]
variables []
